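; Bad News launcher weapon Blueprint: a chargeable weapon that tethers effects to its fired projectile and is discarded when empty.
; Firing activates a tether effect for a tunable time, with Tick feeding it the projectile location, and shakes the camera.
; Cooldown montages and cue are triggered manually after a delay, working around native interruption of fire animations.
; Ammo changes send a gameplay event; empty triggers a throw sound, and local unequip removes the item.
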(event Custom|StopChargeForceFeedback
  (bind _chargeforcefeedback (Variables|Default|GetChargeForceFeedback))
  (Utilities|IsValid _chargeforcefeedback
    (:Is%20Valid
      (ForceFeedback|Stop _chargeforcefeedback))))

(event EventOnStartCharge
  (|CallParentFunction)
  (Components|Activation|Deactivate (Variables|Default|GetNS_BadNews_Tether))
  (Animation|Montage|PlayMontage (Variables|Visual|GetWeaponMesh) (Variables|WeaponAnimation|GetWeaponChargeUpMontage)))

(event EventOnAmmoCountChanged
  ;[1 1 1] On Ammo Changed - have to rep from client to server since no server function available
  (CallFunction|RepAmmoCountChanged))

(event Custom|CheckWeaponAmmo
  ;[1 1 1] Check Weapon Ammo (for discard at 0)
  (if (or (!= (Variables|Default|GetPreviousRemainingAmmo) (FortWeapon|GetRemainingAmmo)) (<= (FortWeapon|GetRemainingAmmo) 0))
    (Variables|Default|SetPreviousRemainingAmmo (FortWeapon|GetRemainingAmmo))
    ;[1 1 1] On Ammo Changed - have to rep from client to server since no server function available
    (bind _asfort_pawn (Utilities|Casting|CastToFortPawn (Game|GetInstigator)))
    (Ability|SendGameplayEventtoActor _asfort_pawn (Variables|Default|GetTag_AmmoChanged))
    ;[1 1 1] Play sound for throwing the gun away after running out of ammo
    (if (<= (FortWeapon|GetRemainingAmmo) 0)
      (CallFunction|PlayWeaponDiscardSound))))

(event Custom|RepAmmoCountChanged
  ;[1 1 1] On Ammo Changed - have to rep from client to server since no server function available
  (bind _asfort_pawn (Utilities|Casting|CastToFortPawn (Game|GetInstigator)))
  (Ability|SendGameplayEventtoActor _asfort_pawn (Variables|Default|GetTag_AmmoChanged))
  ;[1 1 1] Play sound for throwing the gun away after running out of ammo
  (if (<= (FortWeapon|GetRemainingAmmo) 0)
    (CallFunction|PlayWeaponDiscardSound)))

(event Custom|PlayWeaponDiscardSound
  ;[1 1 1] Play sound for throwing the gun away after running out of ammo
  (Variables|Default|GetThrowSound
    (:then)
    (:else
      (bind _returnvalue (Audio|SpawnSoundAttached "/MotherGameplay/Sounds/BadNews/WPN_Bad_News_Drop_Throw.WPN_Bad_News_Drop_Throw" (Variables|Visual|GetWeaponMesh)))
      (Variables|Default|SetThrowSound _returnvalue))))

(event FortWeapon|EventOnLocalUnEquip
  ;[1 1 1] Unequipped - Safety (if unequipped with no ammo...somehow, remove item)
  (Animation|StopAnimMontage (Utilities|Casting|CastToFortPlayerPawn (Game|GetInstigator)) (Variables|Default|GetCharacterCooldownAnim))
  (CallFunction|StopChargeForceFeedback)
  (CallFunction|SetProjectileAttachmentFalse)
  (if (<= (FortWeapon|GetRemainingAmmo) 0)
    (bind _asfort_player_pawn (Utilities|Casting|CastToFortPlayerPawn (Game|GetInstigator)))
    (Inventory|RemoveItemFromPlayerByGuid (Fortnite|GetFortPlayerControllerfromActor _asfort_player_pawn) (FortWeapon|GetInventoryGUID) 1 true)))

(event FortWeapon|EventOnSpawnProjectile
  (params (SpawnProjectile :type Fort%20Projectile%20Base%20Object%20Reference))
  (bind _returnvalue (Game|GetInstigator))
  (bind _asfort_player_pawn (Utilities|Casting|CastToFortPlayerPawn _returnvalue))
  (bind _returnvalue_2 (Animation|GetCurrentMontage _asfort_player_pawn :!class /Script/Engine.Character))
  (bind _returnvalue_3 (Utilities|Equal(Object) _returnvalue_2 "/Game/Animation/Game/MainPlayer/Combat/Guns/GrapplingHook/GrapplingHook_GunToss_M.GrapplingHook_GunToss_M"))
  ;[1 1 1 1 26] Tether
  (|CallParentFunction SpawnProjectile)
  (CallFunction|StopChargeForceFeedback)
  (Utilities|FlowControl|Sequence
    (:then_0
      (Components|Activation|Activate (Variables|Default|GetNS_BadNews_Tether) false)
      (Variables|Default|SetSpawn%20Projectile SpawnProjectile)
      (Variables|Default|SetProjectileAttachmentActive true)
      (bind _returnvalue_4 (Utilities|Time|SetTimerbyEvent (delegate Custom|SetProjectileAttachmentFalse) (ScalableFloat|GetValueatLevel (Variables|Default|GetRow_TetherEffectsTime))))
      (Variables|Default|SetHandleSetProjectileAttachmentFalseTimer _returnvalue_4))
    (:then_1
      (if (> (FortWeapon|GetRemainingAmmo) 0)
        ;;[1 1 1] Manually Calling CoolDown Anims. To account for native issue where cooldown animations interrupt fire animations.
        (if (Networking|IsServer)
          (Utilities|FlowControl|Delay 1.0) ; Manual Delay to allow fire animation to play
          (CallFunction|MultiCastCoolDownAnimFx (Variables|Visual|GetWeaponMesh))
          (elif _returnvalue_3
            _
            (else
              (Utilities|FlowControl|Delay 1.0) ; Manual Delay to allow fire animation to play
              (if _returnvalue_3
                _
                (elif (and (Utilities|Equal(Object) (Variables|/Script/FortniteGame.FortPawn|GetCurrentWeapon _asfort_player_pawn) self) (> (FortWeapon|GetRemainingAmmo) 0))
                  (Animation|Montage|PlayMontage (Variables|Visual|GetWeaponMesh) "/MotherGameplay/Items/BadNews/Assets/Animation/BadNews_CoolDown_Weapon_M.BadNews_CoolDown_Weapon_M" 1.25)
                  (Animation|PlayAnimMontage _asfort_player_pawn (Variables|Default|GetCharacterCooldownAnim) 1.25 "None")
                  (Development|PrintString "COOLDOWN ANIM")
                  (GameplayCue|ExecuteGameplayCue%28Burst%29Local self (Variables|Default|GetCoolDownGCN) (Ability|GameplayCue|MakeGameplayCueParameters 0.0)))))))))
    (:then_2
      (Game|Feedback|ClientStartCameraShake :self (Fortnite|GetFortPlayerControllerfromActor (Game|GetInstigator)) :Shake "/Game/Effects/Fort_Effects/BluePrints/CameraShakes/B_EnergyWeaponFire_CameraShake.B_EnergyWeaponFire_CameraShake_C" :Scale 1.0 :PlaySpace "CameraLocal"))))

(event Custom|SetProjectileAttachmentFalse
  ;[1 1 1 1 26] Tether
  (Variables|Default|SetProjectileAttachmentActive false)
  (Utilities|Time|ClearandInvalidateTimerbyHandle (Variables|Default|GetHandleSetProjectileAttachmentFalseTimer))
  (Components|Activation|Deactivate (Variables|Default|GetNS_BadNews_Tether)))

(event EventTick
  (params (DeltaSeconds :type Float%20%28single-precision%29))
  ;[1 1 1 1 26] Tether
  (|CallParentFunction DeltaSeconds)
  (CallFunction|CheckWeaponAmmo)
  (Utilities|IsValid (Variables|Default|GetSpawn%20Projectile)
    (:Is%20Valid
      (Utilities|IsValid self
        (:Is%20Valid
          (if (Variables|Default|GetProjectileAttachmentActive)
            (Effects|Components|ParticleSystem|SetVectorParameter (Variables|Default|GetNS_BadNews_Tether) "ProjectileLocation" (Transformation|GetActorLocation (Variables|Default|GetSpawn%20Projectile)))
            (else
              (CallFunction|SetProjectileAttachmentFalse))))
        (:Is%20Not%20Valid
          _)))
    (:Is%20Not%20Valid
      _))
  (CallFunction|SetProjectileAttachmentFalse)
  (Variables|Default|GetThrowSound
    (:then
      (Transformation|DetachFromComponent (Variables|Default|GetThrowSound) "KeepRelative" "KeepRelative" "KeepRelative" true))
    (:else)))

(event Custom|MultiCastCoolDownAnimFx
  (params (InSkeletalMeshComponent :type Skeletal%20Mesh%20Component%20Object%20Reference))
  ;[1 1 1 1 26] Tether
  ;;[1 1 1] Manually Calling CoolDown Anims. To account for native issue where cooldown animations interrupt fire animations.
  ;;;[1 1 1] Replicate Cooldown Anims to other Clients
  (Animation|Montage|PlayMontage InSkeletalMeshComponent "/MotherGameplay/Items/BadNews/Assets/Animation/BadNews_CoolDown_Weapon_M.BadNews_CoolDown_Weapon_M" 1.25)
  (Development|PrintString "COOLDOWN ANIM")
  (GameplayCue|ExecuteGameplayCue%28Burst%29Local self (Variables|Default|GetCoolDownGCN) (Ability|GameplayCue|MakeGameplayCueParameters 0.0)))

(fn ConstructionScript
  (|CallParentFunction)
  (Variables|Default|SetProjectileAttachmentActive false))
