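; BadNews ranged weapon projectile: a growing energy orb that periodically zaps nearby pawns and objects.
; A looping server timer gathers targets from two timeline-scaled spheres, line-of-sight checking pawns and optionally hitting obstructing buildings.
; ProcessSomeHits spreads damage and replicated beam FX across frames.
; Impacts stop the orb until the blocking building dies; lifespan and max-range timers, the latter paused while stopped, trigger an expiry warning, then kill.
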(event Custom|ResumeMovement
  (bind _initialvelocity (Variables|UtilityVars|GetInitialVelocity))
  ;[1 1 1] Resume Movement
  (Utilities|Time|ClearandInvalidateTimerbyHandle (Variables|UtilityVars|GetHandle_MaxRangeExpirationTimer))
  (Variables|Default|SetStoppedTargetingOffset "0, 0, 0")
  (Projectile|SetVelocity (Variables|Movement|GetProjectileMovementComponent) _initialvelocity)
  (Niagara|SetNiagaraVariable(Vector3) (Variables|Default|GetNS_BadNews_Projectile) "PositionVelocity" _initialvelocity)
  (CallFunction|SetMaxRangeExpireTimer))

(event Custom|SetMaxRangeExpireTimer
  ;[1 1 1] Set Max Range Expire Timer
  (bind _returnvalue (Utilities|Time|SetTimerbyEvent (EventDispatchers|CreateEvent :!event Custom|BeginExpireWarning) (Variables|UtilityVars|GetTimeTilExpireWarningForMaxRange))) ; Set Max Range Expire Timer
  (Variables|UtilityVars|SetHandle_MaxRangeExpirationTimer _returnvalue)
  (Variables|UtilityVars|SetPreviousGameTime (Utilities|Time|GetGameTimeinSeconds)))

(event Custom|BeginTimers
  ;[1 1 1] BeginTimers
  (bind _returnvalue (Utilities|Time|SetTimerbyEvent (EventDispatchers|CreateEvent :!event Custom|DoTargetSelection) (ScalableFloat|GetValueatLevel (Variables|Balance|GetRow_DamageTickInterval)) true :!pinned true)) ; Start Target Selection Timer (Looping)
  (Variables|UtilityVars|SetHandle_DoTargetSelectionTimer _returnvalue)
  (bind _returnvalue_2 (Utilities|Time|SetTimerbyEvent (EventDispatchers|CreateEvent :!event Custom|BeginExpireWarning) (Math|Float|Max(Float) (- (ScalableFloat|GetValueatLevel (Variables|Balance|GetRow_Lifespan)) (ScalableFloat|GetValueatLevel (Variables|Balance|GetRow_ExpireWarningDuration))) 0.1) :!pinned true)) ; Set Lifespan Expire Timer
  (Variables|UtilityVars|SetHandle_LifespanExpireTimer _returnvalue_2)
  (Variables|UtilityVars|SetTimeTilExpireWarningForMaxRange (Math|Float|Max(Float) (- (/ (ScalableFloat|GetValueatLevel (Variables|Balance|GetRow_MaxRange)) (ScalableFloat|GetValueatLevel (Variables|Balance|GetRow_InitialSpeed))) (ScalableFloat|GetValueatLevel (Variables|Balance|GetRow_ExpireWarningDuration))) 0.1))
  (CallFunction|SetMaxRangeExpireTimer)
  (Actor|SetLifeSpan (+ (ScalableFloat|GetValueatLevel (Variables|Balance|GetRow_Lifespan)) 1.0)) ; Bulletproof
)

(event Custom|InitializeSphereCollision
  (bind _row_capsuleradiusmax (Variables|Balance|GetRow_CapsuleRadiusMax))
  (bind _returnvalue (ScalableFloat|GetValueatLevel _row_capsuleradiusmax))
  (bind _row_capsuleradiusmin (Variables|Balance|GetRow_CapsuleRadiusMin))
  (bind _returnvalue_2 (ScalableFloat|GetValueatLevel _row_capsuleradiusmin))
  (bind _row_sphereradiuspawns_max (Variables|Balance|GetRow_SphereRadiusPawns_Max))
  (bind _returnvalue_3 (ScalableFloat|GetValueatLevel _row_sphereradiuspawns_max))
  (bind _sphereradiusmin_pawns (Variables|UtilityVars|GetSphereRadiusMin_Pawns))
  ;[1 1 1] Initialize Sphere Collision
  (bind _output_get (Variables|UtilityVars|SetSphereRadiusMin_NonPawns (Math|Float|SelectFloat (Variables|/Script/Engine.CapsuleComponent|GetCapsuleRadius (Variables|Collision|GetCapsuleComponent)) (ScalableFloat|GetValueatLevel (Variables|Balance|GetRow_SphereRadiusNonPawns_Min)) (ScalableFloat|GetValueasBool (Variables|Balance|GetRow_UseCapsuleRadiusForSphereMinNonPawns)))))
  (Components|Sphere|SetSphereRadius (Variables|Default|GetTargetCollisionSphere_NonPawns) _output_get false :!pinned true) ; Initialize Sphere Radius
  (bind _output_get_2 (Variables|UtilityVars|SetSphereRadiusMin_Pawns (Math|Float|SelectFloat (Variables|/Script/Engine.CapsuleComponent|GetCapsuleRadius (Variables|Collision|GetCapsuleComponent)) (ScalableFloat|GetValueatLevel (Variables|Balance|GetRow_SphereRadiusPawns_Min)) (ScalableFloat|GetValueasBool (Variables|Balance|GetRow_UseCapsuleRadiusForSphereMinPawns)))))
  (Components|Sphere|SetSphereRadius (Variables|Default|GetTargetCollisionSphere_Pawns) _output_get_2 false :!pinned true) ; Initialize Sphere Radius
  (Components|Capsule|SetCapsuleSize (Variables|Collision|GetCapsuleComponent) (ScalableFloat|GetValueatLevel (Variables|Balance|GetRow_CapsuleRadiusMin)) (ScalableFloat|GetValueatLevel (Variables|Balance|GetRow_CapsuleRadiusMin)) true :!pinned true) ; Initialize Capsule Size
  (bind (_direction _sphereradiusmultiplier) (Timeline|SphereRadiusMultiplier Play)
    (:Update
      (Components|Sphere|SetSphereRadius (Variables|Default|GetTargetCollisionSphere_NonPawns) (Math|Float|Lerp (Variables|UtilityVars|GetSphereRadiusMin_NonPawns) (ScalableFloat|GetValueatLevel (Variables|Balance|GetRow_SphereRadiusNonPawns_Max)) _sphereradiusmultiplier) false :!pinned true) ; Scale Sphere Over Time
      (Components|Sphere|SetSphereRadius (Variables|Default|GetTargetCollisionSphere_Pawns) (Math|Float|Lerp _sphereradiusmin_pawns _returnvalue_3 _sphereradiusmultiplier) false :!pinned true) ; Scale Sphere Over Time
      (Utilities|FlowControl|Sequence
        (:then_0
          (Components|Capsule|SetCapsuleSize (Variables|Collision|GetCapsuleComponent) (Math|Float|Lerp _returnvalue_2 _returnvalue _sphereradiusmultiplier) (Math|Float|Lerp _returnvalue_2 _returnvalue _sphereradiusmultiplier) true :!pinned true) ; Scale Capsule Over Time
          (bind _output_get_3 (Variables|Effects|SetFX_Sphere_Large_Size (Math|Float|Lerp _returnvalue_2 _returnvalue _sphereradiusmultiplier)))
          (Effects|Components|ParticleSystem|SetFloatParameter (Variables|Default|GetNS_BadNews_Projectile) "FX_Sphere_Small_Size" _output_get_3))
        (:then_1
          (bind _output_get_4 (Variables|Effects|SetFX_Sphere_Small_Size (Math|Float|Lerp _sphereradiusmin_pawns _returnvalue_3 _sphereradiusmultiplier)))
          (Effects|Components|ParticleSystem|SetFloatParameter (Variables|Default|GetNS_BadNews_Projectile) "FX_Sphere_Large_Size" _output_get_4))))))

(event Custom|InitializeVariablesAndBeginTimers
  (bind _row_initialspeed (Variables|Balance|GetRow_InitialSpeed))
  (bind _returnvalue (ScalableFloat|GetValueatLevel _row_initialspeed))
  (bind _returnvalue_2 (Transformation|GetActorRotation))
  (bind _returnvalue_3 (Math|Rotator|GetRotationXVector _returnvalue_2))
  (bind _returnvalue_4 (* _returnvalue_3 _returnvalue))
  (bind _row_initialspeed_2 (Variables|Balance|GetRow_InitialSpeed))
  (bind _returnvalue_5 (ScalableFloat|GetValueatLevel _row_initialspeed_2))
  (bind _self self)
  (bind _instigator (Variables|Actor|GetInstigator))
  (bind _array (Utilities|Array|MakeArray _instigator _self))
  (bind _projectilemovementcomponent (Variables|Movement|GetProjectileMovementComponent))
  ;[1 1 1] On Begin Play
  (Variables|/Script/Engine.ProjectileMovementComponent|SetMaxSpeed _returnvalue_5 _projectilemovementcomponent)
  (Variables|/Script/Engine.ProjectileMovementComponent|SetInitialSpeed _returnvalue_5 _projectilemovementcomponent)
  (Projectile|SetVelocity _projectilemovementcomponent _returnvalue_4)
  (bind _output_get (Variables|UtilityVars|SetInitialVelocity _returnvalue_4))
  (Niagara|SetNiagaraVariable(Vector3) (Variables|Default|GetNS_BadNews_Projectile) "PositionVelocity" _output_get)
  (Variables|UtilityVars|SetHitPawns _array)
  (Variables|UtilityVars|SetHitNonPawns _array)
  (Variables|UtilityVars|SetOnlyHitNonPawnsOnce (ScalableFloat|GetValueasBool (Variables|Balance|GetRow_OnlyHitNonPawnsOnce)))
  (Variables|UtilityVars|SetOnlyHitPawnsOnce (ScalableFloat|GetValueasBool (Variables|Balance|GetRow_OnlyHitPawnsOnce)))
  (CallFunction|InitializeSphereCollision)
  (CallFunction|BeginTimers))

(event EventBeginPlay
  ;[1 1 1] Begin Play
  (|CallParentFunction)
  (Networking|CanExecuteCosmeticEvents
    (:True
      (bind _asfort_player_pawn (Utilities|Casting|CastToFortPlayerPawn (Variables|Actor|GetInstigator)))
      (Effects|Components|ParticleSystem|SetVectorParameter (Variables|Default|GetNS_BadNews_Projectile) "Muzzle_Location" (Effects|GetMuzzleLocation (Variables|/Script/FortniteGame.FortPawn|GetCurrentWeapon _asfort_player_pawn) -1)))))

(event Custom|BeginExpireWarning
  (bind _row_expirewarningduration (Variables|Balance|GetRow_ExpireWarningDuration))
  (bind _returnvalue (ScalableFloat|GetValueatLevel _row_expirewarningduration))
  ;[1 1 1] Begin Expire Warning
  (Utilities|FlowControl|DoOnce false
    (:Completed
      (if (> _returnvalue 0.0)
        (Utilities|Time|SetTimerbyEvent (EventDispatchers|CreateEvent :!event Custom|Expire) _returnvalue)
        (bind _returnvalue_2 (Audio|SpawnSoundAttached :Sound (Variables|Sound|GetPreExpireSound) :AttachToComponent (Variables|Default|GetAmbientOrbAudio) :AttachPointName "None" :LocationType "KeepRelativeOffset" :bStopWhenAttachedToDestroyed true :VolumeMultiplier 1.0 :PitchMultiplier 1.0 :StartTime 0.0 :bAutoDestroy true))
        (Variables|Default|SetPreExpireAudioComponent _returnvalue_2)
        (Effects|Components|ParticleSystem|SetBoolParameter (Variables|Default|GetNS_BadNews_Projectile) "ExpireStart" true)
        (else
          (CallFunction|Expire))))))

(event Custom|Expire
  ;[1 1 1] On Expire - Play destruction FX and then kill
  (Effects|Components|ParticleSystem|SetBoolParameter (Variables|Default|GetNS_BadNews_Projectile) "Expired" true)
  (Delivery|Kill)
  (Audio|Components|Audio|FadeOut (Variables|Default|GetAmbientOrbAudio) 0.2 0.0 "Linear" :!class /Script/Engine.AudioComponent)
  (bind _preexpireaudiocomponent (Variables|Default|GetPreExpireAudioComponent)
    (:then
      (Audio|Components|Audio|FadeOut _preexpireaudiocomponent 0.2 0.0 "Linear" :!class /Script/Engine.AudioComponent))
    (:else
      _))
  (Audio|PlaySoundatLocation (Variables|Sound|GetExpireSound) (Transformation|GetActorLocation)))

(event Custom|ShowAudioVisualizer
  ;[1 1 1] Show Audio Visualizer
  (Fortnite|ShowSoundIndicatorLocalforTeam :Location (Transformation|GetActorLocation) :MaxAudibleDistance (ScalableFloat|GetValueatLevel (Variables|Balance|GetRow_AudioVisualizerRadius)) :IndicatorType "Gunshot" :Team (Team|GetActorTeam (Variables|Actor|GetInstigator)) :Affiliations (Variables|Default|GetAffiliations) :Instigator (Variables|Actor|GetInstigator) :Tint "(R=0.000000,G=0.000000,B=0.000000,A=0.000000)" :AttenuationMaxRangeMultiplier 1.0 :AttenuationIntensityMultiplier 1.0))

(event EventDestroyed)

(event OnComponentBeginOverlap(TargetCollisionSphere_NonPawns)
  (params (OverlappedComponent :type Primitive%20Component%20Object%20Reference) (OtherActor :type Actor%20Object%20Reference) (OtherComp :type Primitive%20Component%20Object%20Reference) (OtherBodyIndex :type Integer) (bFromSweep :type Boolean) (SweepResult :type Hit%20Result%20Structure%20%28by%20ref%29)))

(event Custom|TryHitNonPawn
  (bind _hitactor (Variables|UtilityVars|GetHitActor))
  (bind _hitactor_2 (Variables|UtilityVars|GetHitActor))
  ;[1 1 1] On TargetCollisionSphere begin overlap - Make a list of actors to be hit. Hits will be spread out over multiple frames by ProcessSomeHits
  (if (Variables|/Script/Engine.Actor|GetbCanBeDamaged _hitactor_2)
    (bind _asbuilding_gameplay_actor (Utilities|Casting|CastToBuildingGameplayActor _hitactor_2)
      (:then
        (if (Health|HasHealthLeft _asbuilding_gameplay_actor)
          _))
      (:CastFailed
        (Utilities|Casting|CastToFortPickup (Variables|UtilityVars|GetHitActor)
          (:CastFailed
            _)) ; Ignore hits vs. pickup items
      ))
    (if (Variables|UtilityVars|GetOnlyHitNonPawnsOnce)
      (if (not (Utilities|Array|ContainsItem (Variables|UtilityVars|GetHitNonPawns) (Variables|UtilityVars|GetHitActor)))
        (Utilities|Array|AddUnique (Variables|UtilityVars|GetHitNonPawns) (Variables|UtilityVars|GetHitActor)))
      (else
        _))
    (Utilities|Array|Add (Variables|UtilityVars|GetDelayedHitActors) _hitactor)
    (Development|PrintString (Utilities|String|Append "Hit nonpawn: " (Utilities|GetDisplayName _hitactor)) true true "(R=1.000000,G=0.817050,B=0.000000,A=1.000000)" 8.0)))

(event Custom|DoTargetSelection
  (bind _hitactor (Variables|UtilityVars|GetHitActor))
  (bind _map_actortolasthittime (Variables|UtilityVars|GetMap_ActorToLastHitTime))
  (bind (_value _returnvalue) (Utilities|Map|Find _map_actortolasthittime _hitactor))
  (bind _hitactor_2 (Variables|UtilityVars|GetHitActor))
  (bind _hitactor_3 (Variables|UtilityVars|GetHitActor))
  (bind _hitactor_4 (Variables|UtilityVars|GetHitActor))
  ;[1 1 1] Doing overlap checks on ticking timer (not native tick which gets disabled on stop) because we don't have a dedicated code solution for this. Not a significant difference in perf.
  (Networking|SwitchHasAuthority
    (:Authority
      (Utilities|Array|Clear (Variables|UtilityVars|GetDelayedHitActors)) ; Any actors that there hasn't been time to hit yet won't be hit
      (bind (_outactors _returnvalue_2) (Collision|ComponentOverlapActors (Variables|Default|GetTargetCollisionSphere_NonPawns) (Transformation|GetWorldTransform (Variables|Default|GetTargetCollisionSphere_NonPawns)) (Variables|Balance|GetObjectTypesToHit_NonPawns) 0 (Variables|UtilityVars|GetHitNonPawns)))
      (if _returnvalue_2
        (for _array_element _outactors
          ;[1 1 1] On TargetCollisionSphere begin overlap - Make a list of actors to be hit. Hits will be spread out over multiple frames by ProcessSomeHits
          (bind _returnvalue_3 (GameplayTags|HasAnyMatchingGameplayTags(Message) _array_element (Variables|Default|GetTags_IgnoreActorForTargeting)))
          (if (not _returnvalue_3)
            (if (not (Utilities|Array|ContainsItem (Variables|UtilityVars|GetBlockingActors) _array_element))
              (Variables|UtilityVars|SetHitActor _array_element)
              (if (Variables|/Script/Engine.Actor|GetbCanBeDamaged _hitactor_4)
                (bind _asbuilding_gameplay_actor (Utilities|Casting|CastToBuildingGameplayActor _hitactor_4)
                  (:then
                    (if (Health|HasHealthLeft _asbuilding_gameplay_actor)
                      _))
                  (:CastFailed
                    (Utilities|Casting|CastToFortPickup (Variables|UtilityVars|GetHitActor)
                      (:CastFailed
                        _)) ; Ignore hits vs. pickup items
                  ))
                (if (Variables|UtilityVars|GetOnlyHitNonPawnsOnce)
                  (if (not (Utilities|Array|ContainsItem (Variables|UtilityVars|GetHitNonPawns) (Variables|UtilityVars|GetHitActor)))
                    (Utilities|Array|AddUnique (Variables|UtilityVars|GetHitNonPawns) (Variables|UtilityVars|GetHitActor)))
                  (else
                    _))
                (Utilities|Array|Add (Variables|UtilityVars|GetDelayedHitActors) _hitactor_2)
                (Development|PrintString (Utilities|String|Append "Hit nonpawn: " (Utilities|GetDisplayName _hitactor_2)) true true "(R=1.000000,G=0.817050,B=0.000000,A=1.000000)" 8.0)))))
        (CallFunction|ProcessSomeHits))
      (Utilities|FlowControl|Delay (* (ScalableFloat|GetValueatLevel (Variables|Balance|GetRow_DamageTickInterval)) 0.5)) ; Spread out ticks
      (Utilities|Array|Clear (Variables|UtilityVars|GetDelayedHitActors)) ; Any actors that there hasn't been time to hit yet won't be hit
      (Utilities|Array|Clear (Variables|UtilityVars|GetBlockingActors))
      (bind (_outactors_2 _returnvalue_4) (Collision|ComponentOverlapActors (Variables|Default|GetTargetCollisionSphere_Pawns) (Transformation|GetWorldTransform (Variables|Default|GetTargetCollisionSphere_Pawns)) (Variables|Balance|GetObjectTypesToHit_Pawns) 0 (Variables|UtilityVars|GetHitPawns)))
      (if _returnvalue_4
        (for _array_element_2 _outactors_2
          (bind _returnvalue_5 (GameplayTags|HasAnyMatchingGameplayTags(Message) _array_element_2 (Variables|Default|GetTags_IgnoreActorForTargeting)))
          (if (not _returnvalue_5)
            (Variables|UtilityVars|SetHitActor _array_element_2)
            (if (Variables|/Script/Engine.Actor|GetbCanBeDamaged (Variables|UtilityVars|GetHitActor))
              (if (Utilities|Enum|NotEqual(Enum) (Team|GetTeamAffiliation (Variables|UtilityVars|GetHitActor) (Variables|Actor|GetInstigator)) "Friendly")
                (bind _asfort_athena_vehicle (Utilities|Casting|CastToFortAthenaVehicle (Variables|UtilityVars|GetHitActor))
                  (:then
                    (bind _asfort_player_pawn (Utilities|Casting|CastToFortPlayerPawn (Variables|Actor|GetInstigator))
                      (:then
                        (if (not (Utilities|Array|ContainsItem (Vehicle|GetAllPassengers _asfort_athena_vehicle) _asfort_player_pawn))
                          _))
                      (:CastFailed
                        _))
                    (bind _outhit (Collision|LineTraceByChannel (Math|Vector|vector+vector (* (Math|Vector|Normalize (- (Collision|GetActorBounds (Variables|UtilityVars|GetHitActor) true false) (Transformation|GetActorLocation self))) (Components|Capsule|GetScaledCapsuleRadius (Variables|Collision|GetCapsuleComponent))) (Variables|Default|GetStoppedTargetingOffset) (Transformation|GetActorLocation self)) (Collision|GetActorBounds (Variables|UtilityVars|GetHitActor) true false) (Variables|Balance|GetTraceChannelForLOSCheckVsPawns)))
                    (if (or (not (Utilities|IsValid (.HitActor (Collision|BreakHitResult _outhit)) :!class /Script/Engine.KismetSystemLibrary)) (Utilities|Equal(Object) (.HitActor (Collision|BreakHitResult _outhit)) (Variables|UtilityVars|GetHitActor)))
                      (if (Variables|UtilityVars|GetOnlyHitPawnsOnce)
                        (if (not (Utilities|Array|ContainsItem (Variables|UtilityVars|GetHitPawns) (Variables|UtilityVars|GetHitActor)))
                          (Utilities|Array|AddUnique (Variables|UtilityVars|GetHitPawns) (Variables|UtilityVars|GetHitActor)))
                        (else
                          _))
                      (if _returnvalue
                        (if (>= (- (Utilities|Time|GetGameTimeinSeconds) _value) (ScalableFloat|GetValueatLevel (Variables|Balance|GetRow_CooldownBetweenPawnHits)))
                          _)
                        (else
                          _))
                      (Utilities|Map|Add (Variables|UtilityVars|GetMap_ActorToLastHitTime) (Variables|UtilityVars|GetHitActor) (Utilities|Time|GetGameTimeinSeconds))
                      (Utilities|Array|Add (Variables|UtilityVars|GetDelayedHitActors) _hitactor_3)
                      (Development|PrintString (Utilities|String|Append "Hit pawn: " (Utilities|GetDisplayName _hitactor_3)) true true "(R=0.035533,G=1.000000,B=0.000000,A=1.000000)" 8.0)
                      (elif (ScalableFloat|GetValueasBool (Variables|Balance|GetRow_bHitBuildingsObstructingPawnTargets))
                        (if (not (Utilities|Array|ContainsItem (Variables|UtilityVars|GetBlockingActors)))
                          (Utilities|Casting|CastToFortPawn (.HitActor (Collision|BreakHitResult _outhit))
                            (:CastFailed
                              (bind _output_get (Variables|UtilityVars|SetHitActor (.HitActor (Collision|BreakHitResult _outhit))))
                              (Utilities|Array|AddUnique (Variables|UtilityVars|GetBlockingActors) _output_get)
                              (CallFunction|TryHitNonPawn))) ; Don't allow pawns to block LOS to the target
                        ))))
                  (:CastFailed
                    _))))))
        (CallFunction|ProcessSomeHits)))
    (:Remote
      (CallFunction|ShowAudioVisualizer))))

(event Collision|EventHit
  (params (MyComp :type Primitive%20Component%20Object%20Reference) (Other :type Actor%20Object%20Reference) (OtherComp :type Primitive%20Component%20Object%20Reference) (bSelfMoved :type Boolean) (HitLocation :type Vector :default "0, 0, 0") (HitNormal :type Vector :default "0, 0, 0") (NormalImpulse :type Vector :default "0, 0, 0") (Hit :type Hit%20Result%20Structure%20%28by%20ref%29))
  (bind (_bblockinghit _binitialoverlap _time _distance _location _impactpoint _normal _impactnormal _physmat) (Collision|BreakHitResult Hit))
  ;[1 1 1] On Hit
  (|CallParentFunction MyComp Other OtherComp bSelfMoved HitLocation HitNormal NormalImpulse Hit)
  (Components|Movement|StopMovementImmediately (Variables|Movement|GetProjectileMovementComponent))
  (Variables|Default|SetStoppedTargetingOffset (* (Math|Vector|Normalize (- (Transformation|GetActorLocation) HitLocation)) (Variables|Default|GetTargetingOffsetDistance)))
  (Utilities|Time|ClearandInvalidateTimerbyHandle (Variables|UtilityVars|GetHandle_MaxRangeExpirationTimer))
  (Variables|UtilityVars|SetTimeTilExpireWarningForMaxRange (Math|Float|Max(Float) 0.1 (- (Variables|UtilityVars|GetTimeTilExpireWarningForMaxRange) (- (Utilities|Time|GetGameTimeinSeconds) (Variables|UtilityVars|GetPreviousGameTime)))))
  (if (Variables|/Script/Engine.Actor|GetbCanBeDamaged Other)
    (bind _asbuilding_actor (Utilities|Casting|CastToBuildingActor Other)
      (:then
        (EventDispatchers|BindEventtoOnDied _asbuilding_actor (EventDispatchers|CreateEvent :!event Custom|BlockingBuildingDestroyed)))
      (:CastFailed
        _))
    (elif (not (Utilities|Enum|Equal(Enum) (Variables|/Script/PhysicsCore.PhysicalMaterial|GetSurfaceType _physmat) "SurfaceType8"))
      _
      (else
        (Variables|Effects|SetFX_Destroyed (Variables|Effects|GetFX_Destroyed_Water) :!pinned true) ; Set destroyed FX to water version
        (CallFunction|Expire))))
  (if (ScalableFloat|GetValueasBool (Variables|Balance|GetRow_bExpireOnHittingIndestrucible))
    (CallFunction|BeginExpireWarning)))

(event Custom|BlockingBuildingDestroyed
  (params (DamagedActor :type Actor%20Object%20Reference) (Damage :type Float%20%28single-precision%29) (InstigatedBy :type Controller%20Object%20Reference) (DamageCauser :type Actor%20Object%20Reference) (HitLocation :type Vector) (FHitComponent :type Primitive%20Component%20Object%20Reference) (BoneName :type Name) (Momentum :type Vector))
  ;[1 1 1] On Hit
  (CallFunction|ResumeMovement))

(event Custom|ProcessSomeHits
  (bind _delayedhitactors (Variables|UtilityVars|GetDelayedHitActors))
  (bind _returnvalue (Utilities|Array|LastIndex _delayedhitactors))
  (bind _output (Utilities|Array|Get(acopy) _delayedhitactors _returnvalue))
  ;[1 1 1] ProcessSomeHits - Deal damage and play fx, spread over multiple frames for perf
  (Variables|UtilityVars|SetNumHitsToProcessNow (Math|Integer|SelectInt (Math|Integer|Min(Integer) (Utilities|Array|Length (Variables|UtilityVars|GetDelayedHitActors)) (Math|Float|Truncate (ScalableFloat|GetValueatLevel (Variables|UtilityVars|GetRow_HitsToProcessPerFrame)))) (Utilities|Array|Length (Variables|UtilityVars|GetDelayedHitActors)) (> (Math|Float|Truncate (ScalableFloat|GetValueatLevel (Variables|UtilityVars|GetRow_HitsToProcessPerFrame))) 0)))
  (while (and (> (Variables|UtilityVars|GetNumHitsToProcessNow) 0) (> (Utilities|Array|Length (Variables|UtilityVars|GetDelayedHitActors)) 0))
    (Utilities|IsValid _output
      (:Is%20Valid
        (bind _output_get (Variables|Default|SetDelayedHitActor _output))
        (Development|PrintString (Utilities|String|Append "BadNews Targeted " (Utilities|GetDisplayName _output_get)) false)
        (Math|Integer|DecrementInt (Variables|UtilityVars|GetNumHitsToProcessNow)
          (:%20%20
            (Utilities|Array|RemoveIndex (Variables|UtilityVars|GetDelayedHitActors) (Utilities|Array|LastIndex (Variables|UtilityVars|GetDelayedHitActors)))
            (Variables|UtilityVars|SetHitLocation (Collision|GetActorBounds (Variables|Default|GetDelayedHitActor) true false))
            (CallFunction|RepBeamFX (Transformation|GetActorLocation) (Variables|UtilityVars|GetHitLocation))
            (Delivery|DoSingleActorHit (Collision|MakeHitResult false false 0.0 0.0 (Variables|UtilityVars|GetHitLocation) (Variables|UtilityVars|GetHitLocation) "0,0,1" "0,0,1" 0 (Variables|Default|GetDelayedHitActor))) ; Uses Explosion and Additional Explosion hit container in weapon GA
          )))
      (:Is%20Not%20Valid
        (Utilities|Array|RemoveIndex (Variables|UtilityVars|GetDelayedHitActors) (Utilities|Array|LastIndex (Variables|UtilityVars|GetDelayedHitActors))))))
  (if (> (Utilities|Array|Length (Variables|UtilityVars|GetDelayedHitActors)) 0)
    (Utilities|Time|SetTimerforNextTickbyEvent (EventDispatchers|CreateEvent :!event Custom|ProcessSomeHits)) ; Come back and do more next frame
  ))

(event EventOnStop
  (params (Hit :type Hit%20Result%20Structure%20%28by%20ref%29))
  ;[1 1 1] On Stop - Event only fires on the first time it stops, not on subsequent stops
  (|CallParentFunction Hit))

(event Custom|Rep_BeamFX
  (params (StartLocation :type Vector) (HitLocation :type Vector))
  (bind _ns_badnews_projectile (Variables|Default|GetNS_BadNews_Projectile))
  ;[1 1 1] Replicate Beam FX from server hits to client FX
  (Utilities|FlowControl|Sequence
    (:then_0
      (CallFunction|BeamFX StartLocation HitLocation))
    (:then_1
      ;;[1 1 1] Temp - this will need to be done with timer by event
      (Effects|Components|ParticleSystem|SetBoolParameter _ns_badnews_projectile "Hitting" true)
      (Utilities|FlowControl|Delay 0.2)
      (Effects|Components|ParticleSystem|SetBoolParameter _ns_badnews_projectile "Hitting" false))))

(fn BeamFX
  (params (StartLocation :type Vector) (HitLocation :type Vector))
  (bind _returnvalue (Effects|Components|ParticleSystem|SpawnFXAtLocation "/MotherGameplay/Items/BadNews/ProtoAssets/NS_BadNews_SecondaryBeam.NS_BadNews_SecondaryBeam" StartLocation))
  (Effects|Components|ParticleSystem|SetVectorParameter _returnvalue "StartLocation" StartLocation)
  (Effects|Components|ParticleSystem|SetVectorParameter _returnvalue "HitLocation" HitLocation)
  (Effects|Components|ParticleSystem|SpawnFXAtLocation (Variables|Effects|GetFX_Impact) (Variables|UtilityVars|GetHitLocation) (Math|Rotator|FindLookatRotation (Variables|UtilityVars|GetHitLocation) (Variables|GetStartLocation)) (Variables|Effects|GetImpactFXScale) true "AutoRelease")
  (Audio|PlaySoundatLocation (Variables|Sound|GetZapHitSound) (Variables|UtilityVars|GetHitLocation)))

(fn GetMaxLifespan
  (return (+ (ScalableFloat|GetValueatLevel (Variables|Balance|GetRow_Lifespan)) 1.0)))

(fn ConstructionScript
  (|CallParentFunction)
  (CallFunction|InitializeVariablesandBeginTimers))
